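# Translucent UI test material for the MF_UI_AdaptiveColors material function, judging by its name and the function call it holds.
# It offsets texture coordinates by world-position X and Y, each shifted by a Move_X or Move_Y scalar parameter.
# The function call and an Opacity parameter sit alongside, with no links shown from the offset coordinates or to the material outputs.
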Material /BattleRoyaleGameUI/Shared/TestAssets/MF_UI_AdaptiveColor_Test
graph {
 properties {bAutomaticallySetUsageInEditor=false, blendMode=BLEND_Translucent, materialDomain=MD_UI, usedShadingModels=Default Lit}
 nodes [
  n1 = WorldPosition()
  n2 = ComponentMask(None=@n1.XY) {r=true}
  n3 = ScalarParameter() {parameterName=Move_X}
  n4 = ComponentMask(None=@n1.XY) {g=true}
  n5 = ScalarParameter() {parameterName=Move_Y}
  Add(A=TextureCoordinate(), B=AppendVector(A=Add(A=@n2, B=@n3), B=Add(A=@n4, B=@n5)))
  MaterialFunctionCall!(BufferSource) {materialFunction=/BattleRoyaleGameUI/Shared/MF_UI_AdaptiveColors}
  ScalarParameter() {defaultValue=1, parameterName=Opacity}
 ]
}
