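# Test squad list row widget that mimics a Battle Royale HUD squadmate entry: portrait, marker, player name, and health and shield bars.
# All displayed values come from local test variables rather than live player data.
# Construct and design-time PreConstruct share one chain that sets the name, tints the marker, and writes portrait, health and shield into dynamic material parameters.
# The platform icon is collapsed and Tick is empty.
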WidgetBlueprint /BattleRoyaleGameUI/Shared/TestAssets/WBP_TestSquadsList
graphs [EventGraph {nodes [
   n1 = Event PreConstruct()
   > n2 = Branch(Condition=@n1.IsDesignTime)
   Set Text(
    execute=[@n6.then, @n2.then]
    Text=Get PlayerName() -> (PlayerName:Text)
    self=Get CommonTextBlock_PlayerName() -> (CommonTextBlock_PlayerName:Common Text Object Reference)
   )
   > SetBrushTintColor!(
    execute
    self:Image Object Reference=Get Image_Marker() -> (Image_Marker:Image Object Reference)
    TintColor:Slate Color Structure=Get MarkerColor() -> (MarkerColor:Slate Color Structure)
   ) -> (then)
   > n3 = GetDynamicMaterial(self=Get Icon_AvatarSocial() -> (Icon_AvatarSocial:Fort Social Avatar Icon Object Reference))
   > SetTextureParameterValue(self=@n3, ParameterName=SmallPreview, Value=Get Portrait() -> (Portrait:Texture Object Reference))
   > n4 = GetDynamicMaterial(self=Get Image_HealthBar() -> (Image_HealthBar:Image Object Reference))
   > SetScalarParameterValue(self=@n4, ParameterName=Progress, Value=Get health() -> (health:float))
   > n5 = GetDynamicMaterial(self=Get Image_ShieldBar() -> (Image_ShieldBar:Image Object Reference))
   > SetScalarParameterValue(self=@n5, ParameterName=Progress, Value=Get Shield() -> (Shield:float))
   n6 = Event Construct()
   Event Tick()
  ]}]
variables [
 health {type=number, default=1}
 MarkerColor {type=SlateColor, default {colorUseRule=UseColor_Specified, specifiedColor=<rgba 1 1 1 1>}}
 PlayerName {type=string, default=TestPlayerName}
 Portrait {type=Texture, default=/Game/UI/Foundation/Textures/Icons/Heroes/Athena/Soldier/T-Soldier-HID-005-Athena-Commando-M}
 Shield {type=number, default=1}
]
widgets {
 info {inheritedWidgetCount=0, parentClass=&UserWidget, rootWidgetClass=&VerticalBox, widgetCount=18}
 widgets [
  VerticalBox_0 {widgetClassPath=&VerticalBox, bIsVariable=true, widget=VerticalBox_0, properties {bOverride_Cursor=true}}
  Overlay_28 {widgetClassPath=&Overlay, parent=VerticalBox_0, widget=Overlay_28, slot {class=VerticalBoxSlot}}
  Backing {
   widgetClassPath=&Image
   bIsVariable=true
   parent=Overlay_28
   widget=Backing
   properties {brush {
     bIsDynamicallyLoaded=false
     drawAs=Image
     imageType=FullColor
     mirroring=NoMirror
     resourceName=None
     resourceObject=/BattleRoyaleGameUI/HUD/SquadList/Assets/M_UI_SquadList_Backing
     tiling=NoTile
     imageSize=<320 32>
     margin=<ltrb 0 0 0 0>
     outlineSettings {
      bUseBrushTransparency=false
      roundingType=HalfHeightRadius
      width=0
      color {colorUseRule=UseColor_Specified, specifiedColor=<rgba 0 0 0 0>}
      cornerRadii=<0 0 0 0>
     }
     tintColor {colorUseRule=UseColor_Specified, specifiedColor=<rgba 1 1 1 1>}
     uVRegion {bIsValid=false, max=<0 0>, min=<0 0>}
    }}
   slot {class=OverlaySlot, properties {horizontalAlignment=HAlign_Fill, verticalAlignment=VAlign_Fill}}
  }
  Icon_AvatarSocial {
   widgetClassPath=/Script/FortniteUI.FortSocialAvatarIcon
   bIsVariable=true
   parent=Overlay_28
   widget=Icon_AvatarSocial
   properties {
    materialTextureParamName=SmallPreview
    brush {
     bIsDynamicallyLoaded=false
     drawAs=Image
     imageType=FullColor
     mirroring=NoMirror
     resourceName=None
     resourceObject=/BattleRoyaleGameUI/HUD/SquadList/Assets/M_UI_SquadsList_Portrait
     tiling=NoTile
     imageSize=<48 48>
     margin=<ltrb 0 0 0 0>
     outlineSettings {
      bUseBrushTransparency=false
      roundingType=HalfHeightRadius
      width=0
      color {colorUseRule=UseColor_Specified, specifiedColor=<rgba 0 0 0 0>}
      cornerRadii=<0 0 0 0>
     }
     tintColor {colorUseRule=UseColor_Specified, specifiedColor=<rgba 1 1 1 1>}
     uVRegion {bIsValid=false, max=<0 0>, min=<0 0>}
    }
   }
   slot {class=OverlaySlot, properties {verticalAlignment=VAlign_Bottom}}
  }
  VerticalBox {
   widgetClassPath=&VerticalBox
   parent=Overlay_28
   widget=VerticalBox
   slot {class=OverlaySlot, properties {horizontalAlignment=HAlign_Fill, verticalAlignment=VAlign_Center, padding=<ltrb 50 0 16 4>}}
  }
  HorizontalBox_87 {widgetClassPath=&HorizontalBox, parent=VerticalBox, widget=HorizontalBox_87, slot {class=VerticalBoxSlot}}
  Image_Marker {
   widgetClassPath=&Image
   bIsVariable=true
   parent=HorizontalBox_87
   widget=Image_Marker
   properties {brush {
     bIsDynamicallyLoaded=false
     drawAs=Image
     imageType=FullColor
     mirroring=NoMirror
     resourceName=None
     resourceObject=/BattleRoyaleGameUI/HUD/SquadList/Assets/M_UI_AngledBulletPoint
     tiling=NoTile
     imageSize=<10 14>
     margin=<ltrb 0 0 0 0>
     outlineSettings {
      bUseBrushTransparency=false
      roundingType=HalfHeightRadius
      width=0
      color {colorUseRule=UseColor_Specified, specifiedColor=<rgba 0 0 0 0>}
      cornerRadii=<0 0 0 0>
     }
     tintColor {colorUseRule=UseColor_Specified, specifiedColor=<rgba 1 1 1 1>}
     uVRegion {bIsValid=false, max=<0 0>, min=<0 0>}
    }}
   slot {class=HorizontalBoxSlot, properties {horizontalAlignment=HAlign_Left, verticalAlignment=VAlign_Center, padding=<ltrb 0 0 2 0>}}
  }
  OverlayNameandLevel {
   widgetClassPath=&Overlay
   parent=HorizontalBox_87
   widget=OverlayNameandLevel
   slot {class=HorizontalBoxSlot, properties {verticalAlignment=VAlign_Center}}
  }
  CommonTextBlock_PlayerName {
   widgetClassPath=&CommonTextBlock
   bIsVariable=true
   parent=OverlayNameandLevel
   widget=CommonTextBlock_PlayerName
   properties {
    text=TestPlayerName
    font {
     bForceMonospaced=false
     bMaterialIsStencil=false
     fontMaterial=/BattleRoyaleGameUI/HUD/Compass/Assets/M_Font_PP_White
     fontObject=&BurbankSmall
     letterSpacing=0
     monospacedWidth=1
     size=12
     skewAmount=0
     typefaceFontName=Black
     outlineSettings {
      bApplyOutlineToDropShadows=false
      bMiteredCorners=false
      bSeparateFillAlpha=true
      outlineMaterial=/BattleRoyaleGameUI/Shared/MI_UI_Font_SoftShadow_SmallText
      outlineSize=3
      outlineColor=<rgba 1 1 1 1>
     }
    }
    margin=<ltrb -2 -1 -2 0>
    shadowOffset=<0 0>
   }
   slot {class=OverlaySlot, properties {horizontalAlignment=HAlign_Right, verticalAlignment=VAlign_Center}}
  }
  Spacer_111 {widgetClassPath=&Spacer, parent=HorizontalBox_87, widget=Spacer_111, properties {size=<5 1>}, slot {class=HorizontalBoxSlot}}
  SizeBox_0 {
   widgetClassPath=&SizeBox
   parent=HorizontalBox_87
   widget=SizeBox_0
   properties {
    bOverride_MaxDesiredHeight=true
    bOverride_MaxDesiredWidth=true
    heightOverride=22
    maxDesiredHeight=20
    maxDesiredWidth=20
    widthOverride=22
   }
   slot {class=HorizontalBoxSlot, properties {horizontalAlignment=HAlign_Center, verticalAlignment=VAlign_Center, padding=<ltrb 0 -5 0 -5>}}
  }
  Image_PlatformIcon {
   widgetClassPath=&Image
   bIsVariable=true
   parent=SizeBox_0
   widget=Image_PlatformIcon
   properties {
    renderOpacity=0.9
    visibility=Collapsed
    brush {
     bIsDynamicallyLoaded=false
     drawAs=Image
     imageType=NoImage
     mirroring=NoMirror
     resourceName=None
     resourceObject=/Game/UI/Friends_UI/Social/Console_PlatformIcon_32x
     tiling=NoTile
     imageSize=<32 32>
     margin=<ltrb 0 0 0 0>
     outlineSettings {
      bUseBrushTransparency=false
      roundingType=HalfHeightRadius
      width=0
      color {colorUseRule=UseColor_Specified, specifiedColor=<rgba 0 0 0 0>}
      cornerRadii=<0 0 0 0>
     }
     tintColor {colorUseRule=UseColor_Specified, specifiedColor=<rgba 1 1 1 1>}
     uVRegion {bIsValid=false, max=<0 0>, min=<0 0>}
    }
   }
   slot {class=SizeBoxSlot}
  }
  SizeBox_ShieldBar {
   widgetClassPath=&SizeBox
   parent=VerticalBox
   widget=SizeBox_ShieldBar
   properties {bOverride_HeightOverride=true, bOverride_WidthOverride=true, heightOverride=4, widthOverride=172}
   slot {class=VerticalBoxSlot, properties {horizontalAlignment=HAlign_Left}}
  }
  Overlay_ShieldBar {widgetClassPath=&Overlay, bIsVariable=true, parent=SizeBox_ShieldBar, widget=Overlay_ShieldBar, slot {class=SizeBoxSlot}}
  Image_ShieldBar {
   widgetClassPath=&Image
   bIsVariable=true
   parent=Overlay_ShieldBar
   widget=Image_ShieldBar
   properties {brush {
     bIsDynamicallyLoaded=false
     drawAs=Image
     imageType=FullColor
     mirroring=NoMirror
     resourceName=None
     resourceObject=/BattleRoyaleGameUI/HUD/PlayerHitPointInfo/Assets/MI_UI_ShieldBar_Squadmate
     tiling=NoTile
     imageSize=<16 6>
     margin=<ltrb 0 0 0 0>
     outlineSettings {
      bUseBrushTransparency=false
      roundingType=HalfHeightRadius
      width=0
      color {colorUseRule=UseColor_Specified, specifiedColor=<rgba 0 0 0 0>}
      cornerRadii=<0 0 0 0>
     }
     tintColor {colorUseRule=UseColor_Specified, specifiedColor=<rgba 1 1 1 1>}
     uVRegion {bIsValid=false, max=<0 0>, min=<0 0>}
    }}
   slot {class=OverlaySlot, properties {horizontalAlignment=HAlign_Fill, verticalAlignment=VAlign_Fill}}
  }
  SizeBox_HealthBar {
   widgetClassPath=&SizeBox
   parent=VerticalBox
   widget=SizeBox_HealthBar
   properties {bOverride_MaxDesiredHeight=true, bOverride_WidthOverride=true, heightOverride=10, maxDesiredHeight=8, widthOverride=172}
   slot {class=VerticalBoxSlot, properties {horizontalAlignment=HAlign_Left, verticalAlignment=VAlign_Top, padding=<ltrb 0 2 0 0>}}
  }
  Overlay_Healthbar {widgetClassPath=&Overlay, bIsVariable=true, parent=SizeBox_HealthBar, widget=Overlay_Healthbar, slot {class=SizeBoxSlot}}
  Image_HealthBar {
   widgetClassPath=&Image
   bIsVariable=true
   parent=Overlay_Healthbar
   widget=Image_HealthBar
   properties {brush {
     bIsDynamicallyLoaded=false
     drawAs=Image
     imageType=FullColor
     mirroring=NoMirror
     resourceName=None
     resourceObject=/BattleRoyaleGameUI/HUD/PlayerHitPointInfo/Assets/MI_UI_HealthBar_Squadmate
     tiling=NoTile
     imageSize=<16 12>
     margin=<ltrb 0 0 0 0>
     outlineSettings {
      bUseBrushTransparency=false
      roundingType=HalfHeightRadius
      width=0
      color {colorUseRule=UseColor_Specified, specifiedColor=<rgba 0 0 0 0>}
      cornerRadii=<0 0 0 0>
     }
     tintColor {colorUseRule=UseColor_Specified, specifiedColor=<rgba 1 1 1 1>}
     uVRegion {bIsValid=false, max=<0 0>, min=<0 0>}
    }}
   slot {class=OverlaySlot, properties {horizontalAlignment=HAlign_Fill, verticalAlignment=VAlign_Fill}}
  }
 ]
}
